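# PistolCanary_hac1000_gam is a Level Sequence in the PistolCanary hac1000 set that serves only as an event timeline.
# It has no object bindings and a single root event track with one trigger section.
# That section holds two event keys a few frames apart, early in the 60 fps timeline.
# The event structs show no endpoint or payload, so what they trigger is not visible.
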LevelSequence /PistolCanary/Sequence/hac/hac1000/PistolCanary_hac1000_gam
bindings []
display_rate {denominator=1, numerator=60}
tick_resolution {denominator=1, numerator=24000}
tracks [{
  type=MovieSceneEventTrack
  sections [{
    type=MovieSceneEventTriggerSection
    channels [None {
      type=MovieSceneScriptingEventChannel
      keys [[3716, "<Struct 'MovieSceneEvent' (0x852f46ab540) {}>"], [4151, "<Struct 'MovieSceneEvent' (0x852f46ab180) {}>"]]
     }]
   }]
 }]
