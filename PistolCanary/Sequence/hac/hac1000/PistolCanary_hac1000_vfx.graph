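# Level sequence holding the VFX layer of cinematic shot hac1000, which controls one Niagara effect, a Mist Raven dark coil raven-thrower attack.
# The sequence possesses the Niagara actor and its component, with a system spawn section on the component setting the effect's active range.
# Matching keys on component Auto Activate and actor visibility switch both on shortly after that section starts and off at its end.
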LevelSequence /PistolCanary/Sequence/hac/hac1000/PistolCanary_hac1000_vfx
bindings [
 NiagaraComponent0 {
  id=#1
  parent_id=#2
  possessed_class=NiagaraComponent
  tracks [
   {type=MovieSceneNiagaraSystemTrack, sections [{type=MovieSceneNiagaraSystemSpawnSection, end_frame=4170, start_frame=3600}]}
   {
    type=MovieSceneBoolTrack
    display_name=Auto Activate
    sections [{
      type=MovieSceneBoolSection
      channels [None {type=MovieSceneScriptingBoolChannel, default=true, keys [[3732,false],[3742,true],[4170,false]]}]
     }]
   }
  ]
 }
 NS_MistRaven_DarkCoil_RavenThrower_AttackCinematic {
  id=#2
  possessed_class=NiagaraActor
  tracks [{
    type=MovieSceneVisibilityTrack
    sections [{
      type=MovieSceneVisibilitySection
      channels [None {type=MovieSceneScriptingBoolChannel, default=false, keys [[3732,false],[3742,true],[4170,false]]}]
     }]
   }]
 }
]
display_rate {denominator=1, numerator=60}
tick_resolution {denominator=1, numerator=24000}
tracks []
